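# Niagara effect for the spin in the MerryMaking emote, firing bursts of glowing rings and swirling music notes that follow the character's skeletal mesh.
# BurstRing002 spawns five index-staggered mesh rings that expand and fade, driven by dynamic material parameters.
# MusicNotes launches two spawn groups from ring shapes under gravity, drag and vortex force, with user parameters setting spin direction and ring tilt.
# Flare and Shockwave01 emitters are disabled.
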NiagaraSystem /BRCosmetics/Animation/Game/MainPlayer/Emotes/MerryMaking/FX/NS_Emote_MerryMaking_Spin
dependencies {
 usedDataInterfaces [
  &NiagaraDataInterfaceCurve
  &NiagaraDataInterfaceColorCurve
  &NiagaraDataInterfaceMeshRendererInfo
  &NiagaraDataInterfaceSkeletalMesh
 ]
 usedDynamicInputs []
 usedModules [
  &self:SystemSpawnScript.SystemScriptSource.SystemScriptGraph.NiagaraNodeAssignment_0.SetVariables_#1
  &SystemState
 ]
}
emitters {
 BurstRing002 {
  emitterSpawnScript [/SetVariables(Emitter.float:NiagaraFloat=&RandomRangeFloat(Minimum=0, Maximum=500)) {isSetParameters=true}]
  emitterUpdateScript [
   &EmitterState(Life Cycle Mode=::NewEnumerator1, Loop Behavior=::NewEnumerator1, Loop Duration Mode=::NewEnumerator1, UseLoopDelay=false)
   &SpawnBurst_Instantaneous(Spawn Count=5, Use Spawn Probability=false, UseLoopCountLimit=false)
  ]
  particleSpawnScript [
   &InitializeParticle(
    Lifetime=0.75
    Mesh Scale Mode=::NewEnumerator3
    Mesh Uniform Scale=&Multiply_Float(
     A=0.5
     B=&FloatFromCurve(
      FloatCurve=DataInterface() {propertyValues {Curve {keys [
          {interpMode=RCIM_Cubic, leaveTangent=0.7118, tangentMode=RCTM_Break, time=0, value=0.4}
          {arriveTangent=0.486, tangentMode=RCTM_User, time=1, value=1}
         ]}}}
      CurveIndex=&ReturnNormalizedExecIndex()
     )
    )
    Sprite Size Mode=::NewEnumerator3
    UsePositionOffset=true
    Write Material Random=true
   )
   &Color(Color=&MultiplyLinearColor(
     Linear Color=$Particles.Color
     Scale Factor=&MakeVector4(
      X=1
      Y=1
      Z=1
      W=&FloatFromCurve(
       FloatCurve=DataInterface() {propertyValues {Curve {keys [
           {interpMode=RCIM_Cubic, leaveTangent=0.9764, tangentMode=RCTM_Break, time=0, value=0.2}
           {arriveTangent=0.4606, interpMode=RCIM_Cubic, tangentMode=RCTM_Break, time=1, value=1}
          ]}}}
       CurveIndex=&ReturnNormalizedExecIndex()
      )
     )
    ))
   &InitialMeshOrientation(
    Mesh Orientation Mode=::NewEnumerator1
    Rotation=&MakeVector(X=$User.Ring_Rotation_X, Y=0, Z=0)
    Use Rotation Vector=true
   )
  ]
  particleUpdateScript [
   &ParticleState()
   &SkeletalMeshLocation(Bone Sampling Mode=::NewEnumerator4, MaskSpawnGroup=false, Sampled Position Offset=<0 20 0>, Skeletal Mesh=$System.Skeletal Mesh Reader)
   /SetVariables(Particles.MeshOrientation=&FindQuatBetweenTwoVectors(From Vector=<1 0 0>, To Vector=<0 0 0>)) {enabled=false, isSetParameters=true}
   &Timeline(Initial Start Time:NiagaraFloat=$Emitter.float, Play Rate=0.7, Timeline End=100000)
   &DynamicMaterialParameters(
    Index 0 Param 1=&MakeFloatFromVector(Vector=$Particles.Scale, Channel=::NewEnumerator0)
    Index 0 Param 2=2
    Index 0 Param 3=-0.3
    Index 0 Param 4=15
    Index 1 Param 1=$System.Timeline.Time
    Index 1 Param 2=2.0752
    Index 1 Param 3=0.5
    Index 1 Param 4=$Particles.Timeline.Time
    Index 2 Param 1=4
    Index 2 Param 2=2
    Module.Index0Param0Write=true
    Module.Index0Param1Write=true
    Module.Index0Param2Write=true
    Module.Index0Param3Write=true
    Module.Index1Param0Write=true
    Module.Index1Param1Write=true
    Module.Index1Param2Write=true
    Module.Index1Param3Write=true
    Module.Index2Param0Write=true
    Module.Index2Param1Write=true
    Module.Index2Param2Write=true
    Module.Index2Param3Write=true
    Write Parameter Index 1=true
    Write Parameter Index 2=true
   )
   &ScaleMeshSize(Scale Factor=&VectorFromFloat(Value=&FloatFromCurve(
      FloatCurve=DataInterface() {propertyValues {Curve {keys [
          {interpMode=RCIM_Cubic, leaveTangent=1.45, tangentMode=RCTM_Break, time=0, value=0.5}
          {arriveTangent=-0.7179, interpMode=RCIM_Cubic, tangentMode=RCTM_Break, time=1, value=2.5}
         ]}}}
      Scale Curve=&FloatFromCurve(
       FloatCurve=DataInterface() {propertyValues {Curve {keys [<at 0 1.5>,<at 1 1>]}}}
       CurveIndex=&ReturnNormalizedExecIndex()
      )
     )))
   &ScaleColor(
    Scale Alpha=&FloatFromCurve(
     FloatCurve=DataInterface() {propertyValues {Curve {keys [
         {interpMode=RCIM_Cubic, tangentMode=RCTM_User, time=0, value=0}
         {arriveTangent=-0.002, interpMode=RCIM_Cubic, leaveTangent=-0.002, tangentMode=RCTM_User, time=0.2, value=1}
         {tangentMode=RCTM_User, time=1, value=0}
        ]}}}
     Scale Curve=1.45
    )
    Scale RGB=&MakeVectorFromLinearColorRGB(Color=&ColorFromCurve(ColorCurve=DataInterface() {propertyValues {
        AlphaCurve {keys [<at 0 1 RCIM_Cubic>,<at 1 0 RCIM_Cubic>]}
        BlueCurve {keys [<at 0 5 RCIM_Cubic>,<at 1 5 RCIM_Cubic>]}
        GreenCurve {keys [<at 0 5 RCIM_Cubic>,<at 1 0.0251 RCIM_Cubic>]}
        RedCurve {keys [<at 0 0 RCIM_Cubic>,<at 1 0.1155 RCIM_Cubic>]}
       }}))
    ScaleA=true
    ScaleRGB=true
   )
   &ScaleColor(
    Scale Alpha=0.5
    Scale RGB=&VectorFromFloat(Value=&FNTimeOfDayScalar(Night Scale=1, Morning Scale=1, Evening Scale=1))
    ScaleA=true
    ScaleRGB=true
   ) {module=ScaleColor001}
   &CameraOffset(Camera Offset Amount=20)
  ]
  rendererClasses [&NiagaraMeshRendererProperties]
  renderers [{rendererClass=&NiagaraMeshRendererProperties, rendererIndex=0}]
 }
 Flare {
  bEnabled=false
  emitterUpdateScript [
   &EmitterState()
   &SpawnBurst_Instantaneous(Spawn Count=1, Use Spawn Probability=false)
  ]
  particleSpawnScript [
   &InitializeParticle(
    Color=<rgba 5 5 5 0.5>
    Lifetime=0.3
    Sprite Rotation Mode=::NewEnumerator2
    Sprite Size Mode=::NewEnumerator3
    Uniform Sprite Size=350
    UsePositionOffset=false
    Write Material Random=true
   )
   &DynamicMaterialParameters(
    Index 0 Param 1=&RandomRangeFloat(Minimum=0.4, Maximum=0.6)
    Index 0 Param 2=1
    Index 0 Param 3=1
    Index 0 Param 4=1
    Module.Param0WriteEnabled=true
    Module.Param1WriteEnabled=true
    Module.Param2WriteEnabled=true
    Module.Param3WriteEnabled=true
   )
   /SetVariables(Particles.Dark_Glow_CameraOffset=&MakeCustomFloatFromBool(Bool=$User.bIsFrontEndPreview, True Float=1, False Float=50)) {isSetParameters=true}
   &CameraOffset(Camera Offset Amount=&MakeCustomFloatFromBool(Bool=$User.bIsFrontEndPreview, True Float=1, False Float=30))
  ]
  particleUpdateScript [
   /SetVariables(Particles.Dark_Glow=&MakeLinearColorFromVectorAndFloat(
     "Vector (RGB)"=&MakeVectorFromLinearColorRGB(Color=&MultiplyLinearColor(
       Linear Color=&ColorFromCurve(ColorCurve=DataInterface() {propertyValues {
          AlphaCurve {keys []}
          BlueCurve {keys [<at 0.1777 0.45 RCIM_Cubic>,<at 1 0.8594 RCIM_Cubic>]}
          GreenCurve {keys [<at 0.1777 0.2475 RCIM_Cubic>,<at 1 0.4082 RCIM_Cubic>]}
          RedCurve {keys [<at 0.1777 0.045 RCIM_Cubic>,<at 1 0.0859 RCIM_Cubic>]}
         }})
       Scale Factor=&MakeVector4_FromFloat(Float=0.5)
      ))
     "Float (Alpha)"=&Multiply_Float(
      A=&Multiply_Float(A=1, B=1)
      B=&FloatFromCurve(FloatCurve=DataInterface() {propertyValues {Curve {keys [<at 0 0>,<at 0.2 1>,<at 0.7 0.4>,<at 1 0>]}}})
     )
    )) {isSetParameters=true}
   /SetVariables(Particles.Dark_Glow_Size=&Multiply_Vector2(
     A=&Vector2DFromFloat(Value=180)
     B=&Vector2DFromFloat(Value=&FloatFromCurve(
       FloatCurve=DataInterface() {propertyValues {Curve {keys [<at 0 1.5 RCIM_Cubic>,<at 1 0.5>]}}}
       Scale Curve=&FN_FrontEndFloatModifier(Gameplay=1.5, Front End=0.6)
      ))
    )) {isSetParameters=true}
   &ParticleState()
   &SkeletalMeshLocation(Bone Sampling Mode=::NewEnumerator4, MaskSpawnGroup=false, Sampled Position Offset=<0 20 0>, Skeletal Mesh=$System.Skeletal Mesh Reader)
   &ScaleColor(
    Scale Alpha=&Multiply_Float(
     A=1
     B=&FloatFromCurve(FloatCurve=DataInterface() {propertyValues {Curve {keys [<at 0 0>,<at 0.2 1>,<at 0.6 0.6>,<at 0.7 -0.05>]}}})
    )
    Scale RGB=&Multiply_VectorByFloat(
     Vector=&MakeVectorFromLinearColorRGB(Color=&ColorFromCurve(ColorCurve=DataInterface() {propertyValues {
         AlphaCurve {keys []}
         BlueCurve {keys [<at 0.5937 0.7 RCIM_Cubic>,<at 1 1 RCIM_Cubic>]}
         GreenCurve {keys [<at 0.5937 0.1863 RCIM_Cubic>,<at 1 0.6667 RCIM_Cubic>]}
         RedCurve {keys [<at 0.5937 0.035 RCIM_Cubic>,<at 1 0 RCIM_Cubic>]}
        }}))
     Float=5
    )
    ScaleA=true
    ScaleRGB=true
   )
   &ScaleSpriteSize(Scale Factor=&Vector2DFromFloat(Value=&FloatFromCurve(
      FloatCurve=DataInterface() {propertyValues {Curve {keys [<at 0 1.5 RCIM_Cubic>,<at 0.8 0>]}}}
      Scale Curve=&FN_FrontEndFloatModifier(Gameplay=1.1, Front End=0.6)
     )))
   &FN_Light_Attributes(
    Color=<rgba 1.2 1.2 8 1>
    Create Color Attribute=true
    Create Exponent Attribute=true
    Create Position Attribute=true
    Create Radius Attribute=true
    Position=<20 0 0>
    Radius=100
    Scale Color=&Multiply_Float(
     A=3000
     B=&FloatFromCurve(FloatCurve=DataInterface() {propertyValues {Curve {keys [
          {arriveTangent=6.278, interpMode=RCIM_Cubic, leaveTangent=6.278, tangentMode=RCTM_User, time=0, value=0}
          {arriveTangent=-2.0311, interpMode=RCIM_Cubic, leaveTangent=-2.0311, tangentMode=RCTM_User, time=0.5, value=1}
          {arriveTangent=-2.1506, interpMode=RCIM_Cubic, tangentMode=RCTM_Break, time=1, value=0}
         ]}}})
    )
   ) {enabled=false}
  ]
  rendererClasses [&NiagaraSpriteRendererProperties, &NiagaraLightRendererProperties]
  renderers [
   {rendererClass=&NiagaraSpriteRendererProperties}
   {rendererClass=&NiagaraSpriteRendererProperties, rendererIndex=1}
   {rendererClass=&NiagaraLightRendererProperties, rendererIndex=2}
  ]
 }
 MusicNotes {
  emitterUpdateScript [
   &EmitterState()
   &SpawnBurst_Instantaneous(
    Spawn Count=&RandomRangeInt(Minimum=12, Maximum=18)
    Use Spawn Probability=false
    UseLoopCountLimit=false
   )
   &SpawnBurst_Instantaneous(
    Spawn Count=&RandomRangeInt(Minimum=6, Maximum=10)
    Spawn Group=1
    Use Spawn Probability=false
    UseLoopCountLimit=false
   ) {module=SpawnBurst_Instantaneous001}
  ]
  particleSpawnScript [
   /SetVariables(Particles.VisibilityTag=&SpawnGroup()) {isSetParameters=true}
   &InitializeParticle(
    Color Maximum=<rgba 2.2949 0.8989 3 1>
    Color Minimum=<rgba 0.5882 2.1516 3 1>
    Color Mode=::NewEnumerator2
    Lifetime Max=1
    Lifetime Min=0.6
    Lifetime Mode=::NewEnumerator1
    Mesh Renderer Array Visibility Mode=::NewEnumerator3
    Mesh Renderer Info=DataInterface() {propertyValues {MeshRenderer=&self:MusicNotes_0.NiagaraMeshRendererProperties_0}}
    Mesh Scale Mode=::NewEnumerator3
    Mesh Uniform Scale=&MergeHelper_FloatSwitch(
     0=&RandomRangeFloat(Minimum=2.25, Maximum=5)
     1=&RandomRangeFloat(Minimum=1.8, Maximum=3.25)
     2=?
     3=?
     Index=&SpawnGroup()
     Num=::NewEnumerator0
    )
    UsePositionOffset=false
    Write Material Random=true
   )
   &InitialMeshOrientation(Rotation=<0.25 0 0>, Use Rotation Vector=true)
   &ShapeLocation(
    MaskSpawnGroup=false
    Override Seed=false
    Ring / Disc Distribution Mode:ENiagaraSphereDistributionMode=::NewEnumerator1
    Ring Radius=&MergeHelper_FloatSwitch(0=0, 1=30, 2=?, 3=?, Index=&SpawnGroup(), Num=::NewEnumerator0)
    Rotation Angle=90
    Rotation Axis=<0 1 0>
    Rotation Mode=::NewEnumerator1
    Shape Primitive=::NewEnumerator4
    U Position=&MakeCustomFloatFromBool(
     Bool=&SetBoolByFloatComparison(A=$User.Spin_Note_Direction, B=0)
     True Float=&FloatFromCurve(
      FloatCurve=DataInterface() {propertyValues {Curve {keys [<at 0 0.75>,<at 1 0.25>]}}}
      CurveIndex=&ReturnNormalizedExecIndex()
     )
     False Float=&FloatFromCurve(
      FloatCurve=DataInterface() {propertyValues {Curve {keys [<at 0 1.25>,<at 1 0.75>]}}}
      CurveIndex=&ReturnNormalizedExecIndex()
     )
    )
    UseOwnerScale=true
   ) {module=ShapeLocation001}
   &ShapeLocation(
    MaskSpawnGroup=false
    Override Seed=false
    Ring Radius=&MergeHelper_FloatSwitch(0=30, 1=0, 2=?, 3=?, Index=&SpawnGroup(), Num=::NewEnumerator0)
    Rotation Angle=90
    Rotation Axis=<0 1 0>
    Rotation Mode=::NewEnumerator1
    Shape Primitive=::NewEnumerator4
    UseOwnerScale=true
   )
   &AddVelocity(Velocity=&MakeVector(
     X=0
     Y=0
     Z=&MergeHelper_FloatSwitch(
      0=&RandomRangeFloat(Minimum=250, Maximum=475)
      1=&RandomRangeFloat(Minimum=325, Maximum=550)
      2=?
      3=?
      Index=&SpawnGroup()
      Num=::NewEnumerator0
     )
    ))
  ]
  particleUpdateScript [
   &ParticleState()
   &SkeletalMeshLocation(
    Bone Sampling Mode=::NewEnumerator4
    MaskSpawnGroup=false
    Position Sampling=::NewEnumerator1
    Sampled Position Offset=&AddVectorToPosition(Position=$Particles.ShapeLocation.ShapePosition, Vector=<0 20 0>)
    Skeletal Mesh=$System.Skeletal Mesh Reader
   )
   &ScaleMeshSize(Scale Factor=&VectorFromFloat(Value=&FloatFromCurve(FloatCurve=DataInterface() {propertyValues {Curve {
         defaultValue=1
         keys [
          {arriveTangent=9.0725, interpMode=RCIM_Cubic, leaveTangent=9.0725, tangentMode=RCTM_User, time=0, value=0}
          {arriveTangent=-0.0891, interpMode=RCIM_Cubic, leaveTangent=-0.0891, tangentMode=RCTM_User, time=0.1408, value=0.9847}
          {arriveTangent=-0.0405, interpMode=RCIM_Cubic, leaveTangent=-0.0405, tangentMode=RCTM_User, time=1, value=0}
         ]
        }}})))
   &GravityForce(Coordinate Space=::Local, Gravity=<0 0 -750>)
   &VortexForce(
    UseOriginOffset=false
    Vortex Force Amount=&Multiply_Float(
     A=&MergeHelper_FloatSwitch(
      0=&RandomRangeFloat(Minimum=1250, Maximum=1700)
      1=&RandomRangeFloat(Minimum=1800, Maximum=2000)
      2=?
      3=?
      Index=&SpawnGroup()
      Num=::NewEnumerator0
     )
     B=$User.Note_Rotation
    )
    Vortex Origin:NiagaraPosition=$Particles.SkeletalMeshLocation.SampledPosition
   )
   &Drag(Drag=&RandomRangeFloat(Minimum=0.8, Maximum=1.2), Use Linear Drag=true, Use Rotational Drag=true)
   &SolveForcesAndVelocity(Clamp Velocity=false, Limit Acceleration=false)
   &ScaleColor(
    Scale Alpha=&FloatFromCurve(FloatCurve=DataInterface() {propertyValues {Curve {keys [<at 0 1>,<at 1 0>]}}})
    Scale RGB=&VectorFromFloat(Value=&FloatFromCurve(
      FloatCurve=DataInterface() {propertyValues {Curve {
         defaultValue=1
         keys [{interpMode=RCIM_Cubic, tangentMode=RCTM_User, time=0, value=1}, {interpMode=RCIM_Cubic, tangentMode=RCTM_User, time=1, value=0}]
        }}}
      Scale Curve=&FNTimeOfDayScalar(Night Scale=1, Morning Scale=0.85, Day Scale=0.7, Evening Scale=0.85)
     ))
    ScaleA=true
    ScaleRGB=true
   )
  ]
  rendererClasses [&NiagaraMeshRendererProperties]
  renderers [
   {rendererClass=&NiagaraMeshRendererProperties, rendererIndex=0}
   {rendererClass=&NiagaraMeshRendererProperties, rendererIndex=1}
  ]
 }
 Shockwave01 {
  bEnabled=false
  emitterUpdateScript [
   &EmitterState(Life Cycle Mode=::NewEnumerator1, Loop Behavior=::NewEnumerator1, Loop Duration=1, UseLoopDelay=false)
   &SpawnBurst_Instantaneous(Spawn Count=1, Use Spawn Probability=false)
   &SpawnBurst_Instantaneous(Spawn Time=0.2667, Use Spawn Probability=false, UseLoopCountLimit=false) {enabled=false, module=SpawnBurst_Instantaneous001}
  ]
  particleSpawnScript [
   &InitializeParticle(
    Color=<rgba 0.3145 3 1.3804 1>
    Lifetime=0.5
    Mesh Scale Mode=::NewEnumerator3
    Mesh Uniform Scale=0.8
    UsePositionOffset=false
    Write Material Random=true
   )
   &InitialMeshOrientation(Orientation Axis=<0 0 1>, Rotation=<0.25 0 0>, Use Orientation Vector=true, Use Rotation Vector=true)
   &DynamicMaterialParameters(
    Index 0 Param 2=1
    Index 0 Param 3=1
    Index 0 Param 4=1
    Module.Param0WriteEnabled=false
    Module.Param1WriteEnabled=true
    Module.Param2WriteEnabled=true
    Module.Param3WriteEnabled=true
   )
   &InitialMeshOrientation(Orientation Axis=<0 0 1>, Orientation Vector=<0 0 1>, Rotation=<-0.8 0 0>, Use Orientation Vector=true, Use Rotation Vector=true) {enabled=false, module=InitialMeshOrientation001}
  ]
  particleUpdateScript [
   &ParticleState()
   &SkeletalMeshLocation(Bone Sampling Mode=::NewEnumerator4, MaskSpawnGroup=false, Sampled Position Offset=<0 30 0>, Skeletal Mesh=$System.Skeletal Mesh Reader)
   &ScaleMeshSize(Scale Factor=&VectorFromFloat(Value=&FloatFromCurve(FloatCurve=DataInterface() {propertyValues {Curve {keys [<at 0 0.5>,<at 0.2 1>,<at 1 1.2>]}}})))
   &DynamicMaterialParameters(
    Index 0 Param 1=&FloatFromCurve(
     FloatCurve=DataInterface() {propertyValues {Curve {keys [{arriveTangent=0.7396, time=0, value=-1.25}, {arriveTangent=1.2435, time=1, value=0.3}]}}}
     Scale Curve=0.4
    )
    Module.Param0WriteEnabled=true
    Module.Param1WriteEnabled=false
    Module.Param2WriteEnabled=false
    Module.Param3WriteEnabled=false
   ) {module=DynamicMaterialParameters001}
   &ScaleColor(
    Scale Alpha=&FloatFromCurve(FloatCurve=DataInterface() {propertyValues {Curve {keys [
         <at 0 0>
         {interpMode=RCIM_Cubic, leaveTangent=-2.0813, tangentMode=RCTM_Break, time=0.1, value=0.6}
         {arriveTangent=-0.1556, interpMode=RCIM_Cubic, tangentMode=RCTM_Break, time=1, value=0}
        ]}}})
    Scale RGB=&VectorFromFloat(Value=&FloatFromCurve(FloatCurve=DataInterface() {propertyValues {Curve {keys [<at 0 0>,<at 0.3 1>,<at 1 1>]}}}))
    ScaleA=true
    ScaleRGB=true
   )
   &ScaleColor(
    Scale Alpha=&FNTimeOfDayScalar(Night Scale=0.7, Morning Scale=0.3, Evening Scale=0.5)
    ScaleA=true
    ScaleRGB=true
   ) {module=ScaleColor001}
   &MeshRotationForce(Rotation=&MakeVector(
     X=&FloatFromCurve(FloatCurve=DataInterface() {propertyValues {Curve {keys [<at 0 1>,<at 1 0>]}}})
     Y=0
     Z=0
    ))
   &SolveForcesAndVelocity(Clamp Velocity=false, Limit Acceleration=false)
   &CameraOffset(Camera Offset Amount=&MakeCustomFloatFromBool(Bool=$User.bIsFrontEndPreview, True Float=10, False Float=30))
   &ScaleMeshSize(Scale Factor=&MakeCustomVectorFromBool(Bool=$User.bIsFrontEndPreview, True Vector=&VectorFromFloat(Value=0.6), False Vector=<1 1 1>)) {module=ScaleMeshSize001}
  ]
  rendererClasses [&NiagaraMeshRendererProperties]
  renderers [{rendererClass=&NiagaraMeshRendererProperties, rendererIndex=0}]
 }
}
summary {
 systemName=NS_Emote_MerryMaking_Spin
 emitters [
  {bEnabled=false, emitterName=Shockwave01, rendererClasses [&NiagaraMeshRendererProperties]}
  {emitterName=BurstRing002, rendererClasses [&NiagaraMeshRendererProperties]}
  {
   bEnabled=false
   emitterName=Flare
   rendererClasses [&NiagaraSpriteRendererProperties, &NiagaraLightRendererProperties]
  }
  {emitterName=MusicNotes, rendererClasses [&NiagaraMeshRendererProperties]}
 ]
 userVariables [
  User.bIsFrontEnd {type=NiagaraBool}
  User.bIsFrontEndPreview {type=NiagaraBool}
  User.Emote_MerryMaking_Reveal {type=NiagaraFloat, defaultValue=0, description=""}
  User.Note_Rotation {type=NiagaraFloat, defaultValue=1, description=""}
  User.OwnerLocalVelocity {type=Vector3f, description="", defaultValue {struct=Vector3f, value=<0 0 0>}}
  User.Ring_Rotation_X {type=NiagaraFloat, defaultValue=0.25, description=""}
  User.Spin_Note_Direction {type=NiagaraFloat, defaultValue=0, description=""}
 ]
}
