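# Niagara splash effect for the MerryMaking emote: a short burst of expanding rings, a bottom circle and flying music notes.
# RingWaves (with a light) and BotCricle spawn meshes on the first loop only, feeding per-particle accumulated time into dynamic material parameters.
# MusicNotes bursts two spawn groups of meshes from rings, slowed by drag and gravity.
# Brightness or opacity adapts to time of day, and to front-end display for the rings; the Flare emitter is disabled.
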NiagaraSystem /BRCosmetics/Animation/Game/MainPlayer/Emotes/MerryMaking/FX/NS_Emote_MerryMaking_Splash
dependencies {
 usedDataInterfaces [
  &NiagaraDataInterfaceColorCurve
  &NiagaraDataInterfaceCurve
  &NiagaraDataInterfaceMeshRendererInfo
 ]
 usedDynamicInputs [&SetBoolByFloatComparison, &VectorLength]
 usedModules [&SystemState]
}
emitters {
 BotCricle {
  emitterUpdateScript [
   &EmitterState()
   &SpawnBurst_Instantaneous(Spawn Count=1, Spawn Probability=HlslExpression() {hlslExpression="Emitter.LoopCount == 0"}, Use Spawn Probability=true)
  ]
  particleSpawnScript [
   &InitializeParticle(Lifetime=0.5, Mesh Scale Mode=::NewEnumerator3, Mesh Uniform Scale=1.05, UsePositionOffset=false, Write Material Random=true)
   &InitialMeshOrientation(Mesh Axis To Orient=<0 1 0>, Mesh Orientation Mode=::NewEnumerator2, Orientation Vector=<0 0 1>, Use Rotation Vector=false)
   /SetVariables() {isSetParameters=true}
  ]
  particleUpdateScript [
   &ParticleState()
   &ScaleMeshSize(Scale Factor=&VectorFromFloat(Value=&FloatFromCurve(FloatCurve=DataInterface() {propertyValues {Curve {
         defaultValue=1
         keys [
          {arriveTangent=1.1001, interpMode=RCIM_Cubic, leaveTangent=1.1001, tangentMode=RCTM_User, time=0, value=0}
          {arriveTangent=0.0224, interpMode=RCIM_Cubic, leaveTangent=0.0224, tangentMode=RCTM_User, time=1, value=1}
         ]
        }}})))
   /SetVariables(Particles.CustomSpeed:NiagaraFloat=&Add_Float(
     A=$Particles.CustomSpeed
     B=&Multiply_Float(
      A=&Add_Float(A=1, B=&Multiply_Float(A=0.77, B=$User.ParaGlideLeanAlpha))
      B=$Engine.DeltaTime
     )
    )) {isSetParameters=true}
   &DynamicMaterialParameters(
    Index 0 Param 1=&MergeHelper_FloatSwitch(0=50, 1=45, 2=40, 3=?, Index=&SpawnGroup(), Num=::NewEnumerator1)
    Index 0 Param 2=0.08
    Index 0 Param 3=&MergeHelper_FloatSwitch(0=8, 1=1, 2=-8, 3=?, Index=&SpawnGroup(), Num=::NewEnumerator1)
    Index 0 Param 4=3
    Index 1 Param 1=&MergeHelper_FloatSwitch(0=0, 1=0.4, 2=0.7, 3=?, Index=&SpawnGroup(), Num=::NewEnumerator1)
    Index 1 Param 2=&MergeHelper_FloatSwitch(0=0.6, 1=0.8, 2=0.2, 3=?, Index=&SpawnGroup(), Num=::NewEnumerator1)
    Index 1 Param 3=&Add_Float(A=0.15, B=0.05)
    Index 2 Param 1=$Particles.CustomSpeed
    Module.Index0Float0Write=true
    Module.Index0Float1Write=true
    Module.Index0Float2Write=true
    Module.Index0Float3Write=true
    Module.Index1Float0Write=true
    Module.Index1Float1Write=true
    Module.Index1Float2Write=true
    Module.Index1Float3Write=false
    Module.Index2Float0Write=true
    Module.Index2Float1Write=true
    Module.Index2Float2Write=true
    Module.Index2Float3Write=true
    Write Parameter Index 1=::NewEnumerator1
    Write Parameter Index 2=::NewEnumerator1
   )
   &ScaleColor(
    Scale Alpha=&Multiply_Float(
     A=&FNTimeOfDayScalar(Night Scale=0.7, Morning Scale=0.9, Evening Scale=0.8)
     B=&Multiply_Float(
      A=&FN_FrontEndFloatModifier(Gameplay=3, Front End=0.5)
      B=&FloatFromCurve(
       FloatCurve=DataInterface() {propertyValues {Curve {
          defaultValue=1
          keys [
           {arriveTangent=9.0725, interpMode=RCIM_Cubic, leaveTangent=9.0725, tangentMode=RCTM_User, time=0, value=0}
           {arriveTangent=-0.13, interpMode=RCIM_Cubic, leaveTangent=-0.13, tangentMode=RCTM_User, time=0.1408, value=0.9847}
           {arriveTangent=-0.0405, interpMode=RCIM_Cubic, leaveTangent=-0.0405, tangentMode=RCTM_User, time=1, value=0}
          ]
         }}}
       Scale Curve=0.325
      )
     )
    )
    ScaleA=true
    ScaleRGB=true
   )
  ]
  rendererClasses [&NiagaraMeshRendererProperties]
  renderers [{rendererClass=&NiagaraMeshRendererProperties, rendererIndex=0}]
 }
 Flare {
  bEnabled=false
  emitterUpdateScript [
   &EmitterState()
   &SpawnBurst_Instantaneous(Spawn Count=1, Use Spawn Probability=false)
  ]
  particleSpawnScript [
   &InitializeParticle(
    Color=<rgba 5 5 5 0.5>
    Lifetime=0.3
    Sprite Rotation Mode=::NewEnumerator2
    Sprite Size Mode=::NewEnumerator3
    Uniform Sprite Size=350
    UsePositionOffset=false
    Write Material Random=true
   )
   &DynamicMaterialParameters(
    Index 0 Param 1=&RandomRangeFloat(Minimum=0.4, Maximum=0.6)
    Index 0 Param 2=1
    Index 0 Param 3=1
    Index 0 Param 4=1
    Module.Param0WriteEnabled=true
    Module.Param1WriteEnabled=true
    Module.Param2WriteEnabled=true
    Module.Param3WriteEnabled=true
   )
   /SetVariables(Particles.Dark_Glow_CameraOffset=&MakeCustomFloatFromBool(Bool=$User.bIsFrontEndPreview, True Float=1, False Float=50)) {isSetParameters=true}
   &CameraOffset(Camera Offset Amount=&MakeCustomFloatFromBool(Bool=$User.bIsFrontEndPreview, True Float=1, False Float=30))
  ]
  particleUpdateScript [
   /SetVariables(Particles.Dark_Glow=&MakeLinearColorFromVectorAndFloat(
     "Vector (RGB)"=&MakeVectorFromLinearColorRGB(Color=&MultiplyLinearColor(
       Linear Color=&ColorFromCurve(ColorCurve=DataInterface() {propertyValues {
          AlphaCurve {keys []}
          BlueCurve {keys [<at 0.1777 0.45 RCIM_Cubic>,<at 1 0.8594 RCIM_Cubic>]}
          GreenCurve {keys [<at 0.1777 0.2475 RCIM_Cubic>,<at 1 0.4082 RCIM_Cubic>]}
          RedCurve {keys [<at 0.1777 0.045 RCIM_Cubic>,<at 1 0.0859 RCIM_Cubic>]}
         }})
       Scale Factor=&MakeVector4_FromFloat(Float=0.5)
      ))
     "Float (Alpha)"=&Multiply_Float(
      A=&Multiply_Float(A=1, B=1)
      B=&FloatFromCurve(FloatCurve=DataInterface() {propertyValues {Curve {keys [<at 0 0>,<at 0.2 1>,<at 0.7 0.4>,<at 1 0>]}}})
     )
    )) {isSetParameters=true}
   /SetVariables(Particles.Dark_Glow_Size=&Multiply_Vector2(
     A=&Vector2DFromFloat(Value=180)
     B=&Vector2DFromFloat(Value=&FloatFromCurve(
       FloatCurve=DataInterface() {propertyValues {Curve {keys [<at 0 1.5 RCIM_Cubic>,<at 1 0.5>]}}}
       Scale Curve=&FN_FrontEndFloatModifier(Gameplay=1.5, Front End=0.6)
      ))
    )) {isSetParameters=true}
   &ParticleState()
   &ScaleColor(
    Scale Alpha=&Multiply_Float(
     A=1
     B=&FloatFromCurve(FloatCurve=DataInterface() {propertyValues {Curve {keys [<at 0 0>,<at 0.2 1>,<at 0.6 0.6>,<at 0.7 -0.05>]}}})
    )
    Scale RGB=&Multiply_VectorByFloat(
     Vector=&MakeVectorFromLinearColorRGB(Color=&ColorFromCurve(ColorCurve=DataInterface() {propertyValues {
         AlphaCurve {keys []}
         BlueCurve {keys [<at 0.5937 0.7 RCIM_Cubic>,<at 1 1 RCIM_Cubic>]}
         GreenCurve {keys [<at 0.5937 0.1863 RCIM_Cubic>,<at 1 0.6667 RCIM_Cubic>]}
         RedCurve {keys [<at 0.5937 0.035 RCIM_Cubic>,<at 1 0 RCIM_Cubic>]}
        }}))
     Float=5
    )
    ScaleA=true
    ScaleRGB=true
   )
   &ScaleSpriteSize(Scale Factor=&Vector2DFromFloat(Value=&FloatFromCurve(
      FloatCurve=DataInterface() {propertyValues {Curve {keys [<at 0 1.5 RCIM_Cubic>,<at 0.8 0>]}}}
      Scale Curve=&FN_FrontEndFloatModifier(Gameplay=1.1, Front End=0.6)
     )))
   &FN_Light_Attributes(
    Color=<rgba 1.2 1.2 8 1>
    Create Color Attribute=true
    Create Exponent Attribute=true
    Create Position Attribute=true
    Create Radius Attribute=true
    Position=<20 0 0>
    Radius=100
    Scale Color=&Multiply_Float(
     A=3000
     B=&FloatFromCurve(FloatCurve=DataInterface() {propertyValues {Curve {keys [
          {arriveTangent=6.278, interpMode=RCIM_Cubic, leaveTangent=6.278, tangentMode=RCTM_User, time=0, value=0}
          {arriveTangent=-2.0311, interpMode=RCIM_Cubic, leaveTangent=-2.0311, tangentMode=RCTM_User, time=0.5, value=1}
          {arriveTangent=-2.1506, interpMode=RCIM_Cubic, tangentMode=RCTM_Break, time=1, value=0}
         ]}}})
    )
   ) {enabled=false}
  ]
  rendererClasses [&NiagaraSpriteRendererProperties, &NiagaraLightRendererProperties]
  renderers [
   {rendererClass=&NiagaraSpriteRendererProperties}
   {rendererClass=&NiagaraSpriteRendererProperties, rendererIndex=1}
   {rendererClass=&NiagaraLightRendererProperties, rendererIndex=2}
  ]
 }
 MusicNotes {
  emitterUpdateScript [
   &EmitterState()
   &SpawnBurst_Instantaneous(
    Spawn Count=&RandomRangeInt(Minimum=8, Maximum=12)
    Use Spawn Probability=false
    UseLoopCountLimit=false
   )
   &SpawnBurst_Instantaneous(
    Spawn Count=&RandomRangeInt(Minimum=4, Maximum=8)
    Spawn Group=1
    Use Spawn Probability=false
    UseLoopCountLimit=false
   ) {module=SpawnBurst_Instantaneous001}
  ]
  particleSpawnScript [
   /SetVariables(Particles.VisibilityTag=&SpawnGroup()) {isSetParameters=true}
   &InitializeParticle(
    Color Maximum=<rgba 2.2949 0.8989 3 1>
    Color Minimum=<rgba 0.5882 2.1516 3 1>
    Color Mode=::NewEnumerator2
    Lifetime Max=1
    Lifetime Min=0.6
    Lifetime Mode=::NewEnumerator1
    Mesh Renderer Array Visibility Mode=::NewEnumerator3
    Mesh Renderer Info=DataInterface() {propertyValues {MeshRenderer=&self:MusicNotes_0.NiagaraMeshRendererProperties_0}}
    Mesh Scale Mode=::NewEnumerator3
    Mesh Uniform Scale=&MergeHelper_FloatSwitch(
     0=&RandomRangeFloat(Minimum=1.8, Maximum=3.25)
     1=&RandomRangeFloat(Minimum=4, Maximum=6)
     2=?
     3=?
     Index=&SpawnGroup()
     Num=::NewEnumerator0
    )
    UsePositionOffset=false
    Write Material Random=true
   )
   &InitialMeshOrientation(Rotation=<0.25 0 0>, Use Rotation Vector=true)
   &ShapeLocation(
    MaskSpawnGroup=false
    Override Seed=false
    Ring Radius=&MergeHelper_FloatSwitch(0=24, 1=36, 2=?, 3=?, Index=&SpawnGroup(), Num=::NewEnumerator0)
    Shape Primitive=::NewEnumerator4
    UseOwnerScale=true
   )
   &AddVelocity(
    Module.Use Falloff Distance=false
    Origin Offset=&MakeVector(
     X=0
     Y=0
     Z=&MergeHelper_FloatSwitch(0=-80, 1=-60, 2=?, 3=?, Index=&SpawnGroup(), Num=::NewEnumerator0)
    )
    Override Seed=false
    Velocity Mode=::NewEnumerator1
    Velocity Speed=&MergeHelper_FloatSwitch(
     0=&RandomRangeFloat(Minimum=425, Maximum=650)
     1=&RandomRangeFloat(Minimum=425, Maximum=525)
     2=?
     3=?
     Index=&SpawnGroup()
     Num=::NewEnumerator0
    )
   )
  ]
  particleUpdateScript [
   &ParticleState()
   &ScaleMeshSize(Scale Factor=&VectorFromFloat(Value=&FloatFromCurve(FloatCurve=DataInterface() {propertyValues {Curve {
         defaultValue=1
         keys [
          {arriveTangent=9.0725, interpMode=RCIM_Cubic, leaveTangent=9.0725, tangentMode=RCTM_User, time=0, value=0}
          {arriveTangent=-0.0891, interpMode=RCIM_Cubic, leaveTangent=-0.0891, tangentMode=RCTM_User, time=0.1408, value=0.9847}
          {arriveTangent=-0.0405, interpMode=RCIM_Cubic, leaveTangent=-0.0405, tangentMode=RCTM_User, time=1, value=0}
         ]
        }}})))
   &GravityForce()
   &Drag(
    Drag=&MergeHelper_FloatSwitch(
     0=&RandomRangeFloat(Minimum=0.8, Maximum=1.2)
     1=&RandomRangeFloat(Minimum=1.6, Maximum=2.2)
     2=?
     3=?
     Index=&SpawnGroup()
     Num=::NewEnumerator0
    )
    Use Linear Drag=true
    Use Rotational Drag=true
   )
   &SolveForcesAndVelocity(Clamp Velocity=false, Limit Acceleration=false)
   &ScaleColor(
    Scale Alpha=&FloatFromCurve(FloatCurve=DataInterface() {propertyValues {Curve {keys [<at 0 1>,<at 1 0>]}}})
    Scale RGB=&VectorFromFloat(Value=&FloatFromCurve(
      FloatCurve=DataInterface() {propertyValues {Curve {
         defaultValue=1
         keys [{interpMode=RCIM_Cubic, tangentMode=RCTM_User, time=0, value=1}, {interpMode=RCIM_Cubic, tangentMode=RCTM_User, time=1, value=0}]
        }}}
      Scale Curve=&FNTimeOfDayScalar(Night Scale=1, Morning Scale=0.85, Day Scale=0.7, Evening Scale=0.85)
     ))
    ScaleA=true
    ScaleRGB=true
   )
  ]
  rendererClasses [&NiagaraMeshRendererProperties]
  renderers [
   {rendererClass=&NiagaraMeshRendererProperties, rendererIndex=0}
   {rendererClass=&NiagaraMeshRendererProperties, rendererIndex=1}
  ]
 }
 RingWaves {
  emitterUpdateScript [
   &EmitterState()
   &SpawnBurst_Instantaneous(Spawn Count=3, Spawn Probability=HlslExpression() {hlslExpression="Emitter.LoopCount == 0"}, Use Spawn Probability=true)
  ]
  particleSpawnScript [
   &FN_SetSpawnGroup(Number of Spawn Groups=3)
   &InitializeParticle(
    Color=<rgba 0.4783 1 0.8942 1>
    Lifetime=0.5
    Mesh Scale=&Multiply_VectorByFloat(
     Vector=&MergeHelper_VectorSwitch(0=<0.37 0.37 0.25>, 1=<0.37 0.37 0.23>, 2=<0.37 0.37 0.2>, 3=?, Index=&SpawnGroup(), Num=::NewEnumerator1)
     Float=1.15
    )
    Mesh Scale Mode=::NewEnumerator4
    UsePositionOffset=false
    Write Material Random=true
   )
   &InitialMeshOrientation(
    Mesh Orientation Mode=::NewEnumerator2
    Rotation=&MergeHelper_VectorSwitch(0=<0 0 0>, 1=<0 0 0.3>, 2=<0 0 0.8>, 3=?, Index=&SpawnGroup(), Num=::NewEnumerator1)
    Use Rotation Vector=true
   )
   &CameraOffset(Camera Offset Amount=&MergeHelper_FloatSwitch(0=0, 1=0, 2=0, 3=-3, Index=&SpawnGroup(), Num=::NewEnumerator2))
   /SetVariables() {isSetParameters=true}
  ]
  particleUpdateScript [
   &ParticleState()
   &ScaleMeshSize(Scale Factor=&VectorFromFloat(Value=&FloatFromCurve(FloatCurve=DataInterface() {propertyValues {Curve {
         defaultValue=0
         keys [
          {arriveTangent=1.3457, interpMode=RCIM_Cubic, leaveTangent=1.3457, tangentMode=RCTM_User, time=0, value=0}
          {arriveTangent=0.0224, interpMode=RCIM_Cubic, leaveTangent=0.0224, tangentMode=RCTM_User, time=1, value=1}
         ]
        }}})))
   /SetVariables(Particles.CustomSpeed:NiagaraFloat=&Add_Float(
     A=$Particles.CustomSpeed
     B=&Multiply_Float(A=&Add_Float(A=1, B=0.5), B=$Engine.DeltaTime)
    )) {isSetParameters=true}
   &DynamicMaterialParameters(
    Index 0 Param 1=&Add_Float(
     A=&Multiply_Float(
      A=&MergeHelper_FloatSwitch(0=50, 1=45, 2=40, 3=?, Index=&SpawnGroup(), Num=::NewEnumerator1)
      B=0.7
     )
     B=&Multiply_Float(
      A=&MergeHelper_FloatSwitch(0=20, 1=18, 2=16, 3=?, Index=&SpawnGroup(), Num=::NewEnumerator1)
      B=0.525
     )
    )
    Index 0 Param 2=0.2639
    Index 0 Param 3=&MergeHelper_FloatSwitch(0=11, 1=1, 2=-11, 3=8, Index=&SpawnGroup(), Num=::NewEnumerator2)
    Index 1 Param 1=&MergeHelper_FloatSwitch(0=0, 1=0.4, 2=0.7, 3=0, Index=&SpawnGroup(), Num=::NewEnumerator2)
    Index 1 Param 2=&MergeHelper_FloatSwitch(0=0.6, 1=0.8, 2=0.2, 3=0.6, Index=&SpawnGroup(), Num=::NewEnumerator2)
    Index 1 Param 3=$Particles.CustomSpeed
    Index 1 Param 4=&MergeHelper_FloatSwitch(0=0, 1=0, 2=0, 3=1, Index=&SpawnGroup(), Num=::NewEnumerator2)
    Module.Index0Float0Write=true
    Module.Index0Float1Write=true
    Module.Index0Float2Write=true
    Module.Index0Float3Write=true
    Module.Index1Float0Write=true
    Module.Index1Float1Write=true
    Module.Index1Float2Write=true
    Module.Index1Float3Write=true
    Write Parameter Index 1=::NewEnumerator1
   )
   &ScaleColor(
    Scale Alpha=&Multiply_Float(
     A=&FNTimeOfDayScalar(Night Scale=0.7, Morning Scale=0.9, Evening Scale=0.8)
     B=&Multiply_Float(
      A=&FN_FrontEndFloatModifier(Gameplay=1, Front End=0.3)
      B=&FloatFromCurve(
       FloatCurve=DataInterface() {propertyValues {Curve {
          defaultValue=1
          keys [
           {arriveTangent=9.0725, interpMode=RCIM_Cubic, leaveTangent=9.0725, tangentMode=RCTM_User, time=0, value=0}
           {arriveTangent=-0.0696, interpMode=RCIM_Cubic, leaveTangent=-0.0696, tangentMode=RCTM_User, time=0.1408, value=0.9847}
           {arriveTangent=-0.0405, interpMode=RCIM_Cubic, leaveTangent=-0.0405, tangentMode=RCTM_User, time=1, value=0}
          ]
         }}}
       Scale Curve=0.4
      )
     )
    )
    Scale RGB=&VectorFromFloat(Value=&FN_FrontEndFloatModifier(Gameplay=1.2, Front End=1.5))
    ScaleA=true
    ScaleRGB=true
   )
   &FN_Light_Attributes(
    Color=<rgba 0.1104 0.8754 1 1>
    Create Color Attribute=true
    Create Exponent Attribute=true
    Create Position Attribute=true
    Create Radius Attribute=true
    Exponent=4
    Position=<0 0 10>
    Radius=120
    Scale Color=&FloatFromCurve(
     FloatCurve=DataInterface() {propertyValues {Curve {
        defaultValue=1
        keys [
         {arriveTangent=9.0725, interpMode=RCIM_Cubic, leaveTangent=9.0725, tangentMode=RCTM_User, time=0, value=0}
         {arriveTangent=0.2683, interpMode=RCIM_Cubic, leaveTangent=0.2683, tangentMode=RCTM_User, time=0.2, value=1}
         {arriveTangent=0.0034, interpMode=RCIM_Cubic, leaveTangent=0.0034, tangentMode=RCTM_User, time=0.45, value=0.9847}
         {arriveTangent=-0.0405, interpMode=RCIM_Cubic, leaveTangent=-0.0405, tangentMode=RCTM_User, time=1, value=0}
        ]
       }}}
     Scale Curve=4
    )
   )
  ]
  rendererClasses [&NiagaraMeshRendererProperties, &NiagaraLightRendererProperties]
  renderers [
   {rendererClass=&NiagaraMeshRendererProperties, rendererIndex=0}
   {rendererClass=&NiagaraLightRendererProperties, rendererIndex=1}
  ]
 }
}
summary {
 systemName=NS_Emote_MerryMaking_Splash
 emitters [
  {
   bEnabled=false
   emitterName=Flare
   rendererClasses [&NiagaraSpriteRendererProperties, &NiagaraLightRendererProperties]
  }
  {emitterName=RingWaves, rendererClasses [&NiagaraMeshRendererProperties, &NiagaraLightRendererProperties]}
  {emitterName=BotCricle, rendererClasses [&NiagaraMeshRendererProperties]}
  {emitterName=MusicNotes, rendererClasses [&NiagaraMeshRendererProperties]}
 ]
 userVariables [
  User.bIsFrontEnd {type=NiagaraBool}
  User.bIsFrontEndPreview {type=NiagaraBool}
  User.Emote_MerryMaking_Reveal {type=NiagaraFloat, defaultValue=0, description=""}
  User.FEOffset {type=NiagaraFloat, defaultValue=10, description=""}
  User.FEScale {type=NiagaraFloat, defaultValue=0.6, description=""}
  User.OwnerLocalVelocity {type=Vector3f, description="", defaultValue {struct=Vector3f, value=<0 0 0>}}
  User.ParaGlideLeanAlpha {type=NiagaraFloat, defaultValue=0, description=""}
 ]
}
