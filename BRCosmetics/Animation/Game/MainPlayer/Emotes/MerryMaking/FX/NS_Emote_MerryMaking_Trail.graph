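# Niagara system for the MerryMaking emote that draws glowing music trails, wake meshes and a light decal around the character.
# Two anim-trail ribbon emitters shorten their lifetime and gain backward drift based on owner velocity.
# WakeLines builds two partitioned ribbons offset from the skeletal mesh, with sine wobble and motion-driven bending.
# Reveal and Light user floats drive fades and decal intensity; Trail, MusicNotes and debug emitters are disabled.
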NiagaraSystem /BRCosmetics/Animation/Game/MainPlayer/Emotes/MerryMaking/FX/NS_Emote_MerryMaking_Trail
dependencies {
 usedDataInterfaces [
  &NiagaraDataInterfaceColorCurve
  &NiagaraDataInterfaceCurve
  &NiagaraDataInterfaceMeshRendererInfo
  &NiagaraDataInterfaceSkeletalMesh
  &NiagaraDataInterfaceSocketReader
 ]
 usedDynamicInputs []
 usedModules [
  &self:SystemSpawnScript.SystemScriptSource.SystemScriptGraph.NiagaraNodeAssignment_0.SetVariables_#1
  &SystemState
  &BlendSpaceHelper
  &self:SystemSpawnScript.SystemScriptSource.SystemScriptGraph.NiagaraNodeAssignment_1.SetVariables_#2
 ]
}
emitters {
 Decal {
  emitterUpdateScript [
   &EmitterState(Life Cycle Mode=::NewEnumerator1, Loop Behavior=::NewEnumerator1, Loop Duration Mode=::NewEnumerator1, UseLoopDelay=false)
   &SpawnBurst_Instantaneous(Use Spawn Probability=false, UseLoopCountLimit=false)
  ]
  particleSpawnScript [
   &InitializeParticle(Lifetime=0.8, UsePositionOffset=false, Write Material Random=true)
   /SetVariables(Particles.Glow DecalSize:Vector3f=&VectorFromFloat(Value=425)) {isSetParameters=true}
  ]
  particleUpdateScript [
   &ParticleState(
    Kill Particles When Lifetime Has Elapsed=false
    Let Infinitely Lived Particles Die When Emitter Deactivates=true
    Loop Particles Lifetime=true
   )
   &SkeletalMeshLocation(Bone Sampling Mode=::NewEnumerator4, MaskSpawnGroup=false, Skeletal Mesh=$System.Skeletal Mesh Reader)
   &SolveForcesAndVelocity(Clamp Velocity=false, Limit Acceleration=false)
   /SetVariables(Particles.Glow DecalColor:LinearColor=&MakeLinearColorFromVectorAndFloat(
     "Vector (RGB)"=&MakeVectorFromLinearColorRGB(Color=<rgba 0.8245 4.3101 8 1>)
     "Float (Alpha)"=&Multiply_Float(
      A=&Multiply_Float(
       A=&FN_FrontEndFloatModifier(Gameplay=0.13, Front End=0.15)
       B=&FNTimeOfDayScalar(Night Scale=0.55, Morning Scale=0.8, Evening Scale=0.8)
      )
      B=&Multiply_Float(A=0.7, B=$User.Emote_MerryMaking_Light)
     )
    )) {isSetParameters=true}
  ]
  rendererClasses [&NiagaraDecalRendererProperties]
  renderers [
   {rendererClass=&NiagaraDecalRendererProperties, rendererIndex=0}
   {rendererClass=&NiagaraDecalRendererProperties, rendererIndex=1}
  ]
 }
 FN_Debug {
  bEnabled=false
  emitterUpdateScript [
   &EmitterState()
   &SpawnBurst_Instantaneous(Spawn Count=1, Use Spawn Probability=false)
  ]
  particleSpawnScript [/Niagara/Modules/Spawn/Initialization/InitializeParticle(
    Lifetime=5
    Module.Write Color=true
    Module.Write Lifetime=true
    Module.Write Mass=true
    Module.Write Material Random=true
    Module.Write Position=false
    Module.Write Scale=false
    Module.Write SpriteRotation=false
    Module.Write SpriteSize=true
    Position=&SimulationPosition()
    Sprite Size=<100 100>
   )]
  particleUpdateScript [
   &ParticleState()
   /SetVariables(Particles.Debug:NiagaraFloat=&VectorLength(Vector=$User.OwnerLocalVelocity)) {isSetParameters=true}
   &DynamicMaterialParameters(
    Index 0 Param 1=$Particles.Debug
    Module.Param0WriteEnabled=true
    Module.Param1WriteEnabled=true
    Module.Param2WriteEnabled=true
    Module.Param3WriteEnabled=true
   )
  ]
  rendererClasses [&NiagaraSpriteRendererProperties]
  renderers [{rendererClass=&NiagaraSpriteRendererProperties}]
 }
 MusicNotes {
  bEnabled=false
  emitterUpdateScript [
   &EmitterState(Life Cycle Mode=::NewEnumerator1, Loop Behavior=::NewEnumerator1, Loop Duration Mode=::NewEnumerator1, UseLoopDelay=false)
   &SpawnRate(
    SpawnRate=&Multiply_Float(A=8, B=&OneMinusFloat(Float=$User.Emote_MerryMaking_Reveal))
    Use Spawn Probability=false
   )
  ]
  particleSpawnScript [
   /Niagara/Modules/Spawn/Initialization/V2/InitializeParticle(
    Color=<rgba 0 0 0 1>
    Lifetime Max=1.4
    Lifetime Min=0.6
    Lifetime Mode=::NewEnumerator1
    Mesh Renderer Array Visibility Mode=::NewEnumerator3
    Mesh Renderer Info=DataInterface() {propertyValues {MeshRenderer=&self:MusicNotes_0.NiagaraMeshRendererProperties_0}}
    Mesh Scale Mode=::NewEnumerator1
    Mesh Uniform Scale Max=5
    Mesh Uniform Scale Min=3
    UsePositionOffset=false
    Write Material Random=true
   )
   &SkeletalMeshLocation(
    Bone Sampling Mode=::NewEnumerator4
    MaskSpawnGroup=false
    Orientation Sampling=::NewEnumerator1
    Position Sampling=::NewEnumerator1
    Sampled Position Offset=<-20 0 0>
    Skeletal Mesh=$System.Skeletal Mesh Reader
   )
   &ShapeLocation(
    Box Size=<25 0 12>
    MaskSpawnGroup=false
    Override Seed=false
    Shape Origin=&AddVectorToPosition(Position=$Particles.SkeletalMeshLocation.SampledPosition, Vector=<0 0 0>)
    Shape Primitive=::NewEnumerator2
    Surface Only=false
    UseOwnerScale=true
   )
   &AddVelocity(Velocity=&MakeVector(
     X=0
     Y=&MakeCustomFloatFromBool(
      Bool=&SetBoolByFloatComparison(A=&VectorLength(Vector=$User.OwnerLocalVelocity), B=500)
      True Float=&RandomRangeFloat(Minimum=-120, Maximum=-225)
      False Float=&RandomRangeFloat(Minimum=-225, Maximum=-400)
     )
     Z=0
    ))
   &InitialMeshOrientation(
    Mesh Axis To Orient=<-1 0 0>
    Mesh Orientation Mode=::NewEnumerator2
    Orientation Vector=&RotateVectorByQuaternion(Vector To Rotate=<1 0 0>, Quaternion=$Particles.SkeletalMeshLocation.SampledRotation)
    Rotation=<0.25 0 0>
    Use Rotation Vector=true
   )
  ]
  particleUpdateScript [
   &ParticleState()
   &UpdateMeshOrientation(Facing Interpolation=false, Orientation Method=::NewEnumerator1, Override Quat=false, TransformFacing=false, TransformUp=false) {enabled=false}
   &ScaleMeshSize(Scale Factor=&VectorFromFloat(Value=&FloatFromCurve(FloatCurve=DataInterface() {propertyValues {Curve {
         defaultValue=1
         keys [
          {arriveTangent=9.0725, interpMode=RCIM_Cubic, leaveTangent=9.0725, tangentMode=RCTM_User, time=0, value=0}
          {arriveTangent=-2.5249, interpMode=RCIM_Cubic, leaveTangent=-2.5249, tangentMode=RCTM_User, time=0.1408, value=0.9847}
          {arriveTangent=-0.0405, interpMode=RCIM_Cubic, leaveTangent=-0.0405, tangentMode=RCTM_User, time=1, value=0}
         ]
        }}})))
   &SolveForcesAndVelocity(Clamp Velocity=false, Limit Acceleration=false)
   &ScaleColor(
    Scale Alpha=&FloatFromCurve(FloatCurve=DataInterface() {propertyValues {Curve {keys [<at 0 1>,<at 1 0>]}}})
    ScaleA=true
    ScaleRGB=true
   )
   &BlendSpace("Position Blend (World <-> Local)":NiagaraFloat=&Multiply_Float(
     A=&FloatFromCurve(FloatCurve=DataInterface() {propertyValues {Curve {
         defaultValue=1
         keys [
          {arriveTangent=-1.3963, interpMode=RCIM_Cubic, leaveTangent=-1.3963, tangentMode=RCTM_User, time=0, value=1}
          {interpMode=RCIM_Cubic, leaveTangent=-8.0089, time=1, value=0}
         ]
        }}})
     B=&FloatFromCurve(
      FloatCurve=DataInterface() {propertyValues {Curve {keys [<at 0 0 RCIM_Constant>,<at 800 1 RCIM_Constant>]}}}
      CurveIndex=&VectorLength(Vector=$User.OwnerLocalVelocity)
     )
     Translation Contribution:NiagaraFloat=1
     Rotation Contribution:NiagaraFloat=0
    ))
  ]
  rendererClasses [&NiagaraMeshRendererProperties]
  renderers [{rendererClass=&NiagaraMeshRendererProperties, rendererIndex=0}]
 }
 MusicTrail {
  emitterSpawnScript [&FN_RibbonIDOffsetUtility()]
  emitterUpdateScript [
   &EmitterState()
   &FN_SpawnAnimTrail()
  ]
  particleSpawnScript [
   &FN_InitializeAnimTrailBoneIndecies()
   &FN_InitializeAnimTrailBoneIndecies(AnimTrail Selection=1, First Bone or Socket Index=2, Second Bone or Socket Index=3) {enabled=false, module=FN_InitializeAnimTrailBoneIndecies001}
   &FN_InitializeAnimTrail(Lifetime=&FloatFromCurve(
     FloatCurve=DataInterface() {propertyValues {Curve {keys [<at 0 0.825 RCIM_Constant>,<at 800 0.45 RCIM_Constant>]}}}
     CurveIndex=&VectorLength(Vector=$User.OwnerLocalVelocity)
    ))
   &AddVelocity(Velocity=&MakeVector(
     X=0
     Y=&MakeCustomFloatFromBool(
      Bool=&SetBoolByFloatComparison(A=&VectorLength(Vector=$User.OwnerLocalVelocity), B=1)
      True Float=0
      False Float=-400
     )
     Z=0
    ))
  ]
  particleUpdateScript [
   &SolveForcesAndVelocity(Clamp Velocity=false, Limit Acceleration=false)
   &DynamicMaterialParameters(
    Index 0 Param 1=$User.Emote_MerryMaking_Reveal
    Module.Index0Float0Write=true
    Module.Index0Float1Write=true
    Module.Index0Float2Write=true
    Module.Index0Float3Write=true
   )
   &ScaleColor(
    Scale Alpha=&FloatFromCurve(FloatCurve=DataInterface() {propertyValues {Curve {
        defaultValue=1
        keys [
         <at 0 0 RCIM_Cubic>
         {arriveTangentWeight=0.3333, interpMode=RCIM_Cubic, leaveTangentWeight=0.3333, tangentMode=RCTM_User, time=0.125, value=1}
         {arriveTangentWeight=0.3333, interpMode=RCIM_Cubic, leaveTangentWeight=0.3333, tangentMode=RCTM_User, time=0.8, value=0}
        ]
       }}})
    Scale RGB=&Multiply_VectorByFloat(
     Vector=&MakeVectorFromLinearColorRGB(Color=&ColorFromCurve(ColorCurve=DataInterface() {propertyValues {
         AlphaCurve {keys []}
         BlueCurve {keys [<at 0 1 RCIM_Cubic>,<at 0.25 1>,<at 0.5 1>,<at 0.75 1>,<at 1 1>]}
         GreenCurve {keys [<at 0 0.6511 RCIM_Cubic>,<at 0.25 0.3466>,<at 0.5 0.0659>,<at 0.75 0.1016>,<at 1 0>]}
         RedCurve {keys [<at 0 0 RCIM_Cubic>,<at 0.25 0.1017>,<at 0.5 0.2007>,<at 0.75 0.8027>,<at 1 0.7811>]}
        }}))
     Float=25
    )
    ScaleA=true
    ScaleRGB=true
   )
   &FN_Sprite_Attributes(
    Color=&MakeLinearColorFromVectorAndFloat(
     "Vector (RGB)"=&MakeVectorFromLinearColorRGB(Color=&ColorFromCurve(ColorCurve=DataInterface() {propertyValues {
         AlphaCurve {keys []}
         BlueCurve {keys [<at 0 1 RCIM_Cubic>,<at 0.25 1>,<at 0.5 1>,<at 0.75 1>,<at 1 1>]}
         GreenCurve {keys [<at 0 0.6511 RCIM_Cubic>,<at 0.25 0.3466>,<at 0.5 0.0659>,<at 0.75 0.1016>,<at 1 0>]}
         RedCurve {keys [<at 0 0 RCIM_Cubic>,<at 0.25 0.1017>,<at 0.5 0.2007>,<at 0.75 0.8027>,<at 1 0.7811>]}
        }}))
     "Float (Alpha)"=&FloatFromCurve(
      FloatCurve=DataInterface() {propertyValues {Curve {
         defaultValue=1
         keys [
          <at 0 0 RCIM_Cubic>
          {arriveTangentWeight=0.3333, interpMode=RCIM_Cubic, leaveTangentWeight=0.3333, tangentMode=RCTM_User, time=0.125, value=1}
          {arriveTangentWeight=0.3333, interpMode=RCIM_Cubic, leaveTangentWeight=0.3333, tangentMode=RCTM_User, time=0.8, value=0}
         ]
        }}}
      Scale Curve=2
     )
    )
    Create Color Attribute=true
    Write to Module Namespace=true
   )
   &FN_Sprite_Attributes(
    Color=&MakeLinearColorFromVectorAndFloat(
     "Vector (RGB)"=&MakeVectorFromLinearColorRGB(Color=&ColorFromCurve(ColorCurve=DataInterface() {propertyValues {
         AlphaCurve {keys []}
         BlueCurve {keys [<at 0 1 RCIM_Cubic>,<at 0.25 1>,<at 0.5 1>,<at 0.75 1>,<at 1 1>]}
         GreenCurve {keys [<at 0 0.6511 RCIM_Cubic>,<at 0.25 0.3466>,<at 0.5 0.0659>,<at 0.75 0.1016>,<at 1 0>]}
         RedCurve {keys [<at 0 0 RCIM_Cubic>,<at 0.25 0.1017>,<at 0.5 0.2007>,<at 0.75 0.8027>,<at 1 0.7811>]}
        }}))
     "Float (Alpha)"=&FloatFromCurve(
      FloatCurve=DataInterface() {propertyValues {Curve {
         defaultValue=1
         keys [
          <at 0 0 RCIM_Cubic>
          {arriveTangentWeight=0.3333, interpMode=RCIM_Cubic, leaveTangentWeight=0.3333, tangentMode=RCTM_User, time=0.125, value=1}
          {arriveTangentWeight=0.3333, interpMode=RCIM_Cubic, leaveTangentWeight=0.3333, tangentMode=RCTM_User, time=1, value=0}
         ]
        }}}
      Scale Curve=0.8
     )
    )
    Create Color Attribute=true
    Write to Module Namespace=true
   ) {module=FN_Sprite_Attributes001}
   /SetVariables(Particles.LineRibbonWidth:NiagaraFloat=&FloatFromCurve(FloatCurve=DataInterface() {propertyValues {Curve {keys [<at 0 1 RCIM_Cubic>,<at 1 0.8 RCIM_Cubic>]}}}, Scale Curve=75)) {isSetParameters=true}
   /SetVariables(Particles.GlowRibbonWidth:NiagaraFloat=&FloatFromCurve(
     FloatCurve=DataInterface() {propertyValues {Curve {
        defaultValue=1
        keys [
         {interpMode=RCIM_Cubic, tangentMode=RCTM_User, time=0, value=1}
         {arriveTangent=-3.452, interpMode=RCIM_Cubic, leaveTangent=-3.452, tangentMode=RCTM_User, time=1, value=0}
        ]
       }}}
     Scale Curve=100
    )) {isSetParameters=true}
   &FN_UpdateAnimTrail() {module=FN_UpdateAnimTrail001}
   /SetVariables(
    Particles.NotesColor:LinearColor=<rgba 0 0 0 2.5>
    Particles.NotesDP:Vector4f=&MakeVector4(X=1, Y=1, Z=2, W=&Multiply_Float(A=$Emitter.Age, B=-1))
    Particles.NotesDP2:Vector4f=<0 0 0 0>
   ) {isSetParameters=true, module=SetVariables_#3}
  ]
  rendererClasses [&NiagaraRibbonRendererProperties]
  renderers [
   {rendererClass=&NiagaraRibbonRendererProperties}
   {rendererClass=&NiagaraRibbonRendererProperties, rendererIndex=1}
   {rendererClass=&NiagaraRibbonRendererProperties, rendererIndex=2}
   {rendererClass=&NiagaraRibbonRendererProperties, rendererIndex=3}
   {rendererClass=&NiagaraRibbonRendererProperties, rendererIndex=4}
   {rendererClass=&NiagaraRibbonRendererProperties, rendererIndex=5}
  ]
 }
 MusicTrail001 {
  emitterSpawnScript [&FN_RibbonIDOffsetUtility()]
  emitterUpdateScript [
   &EmitterState()
   &FN_SpawnAnimTrail()
  ]
  particleSpawnScript [
   &FN_InitializeAnimTrailBoneIndecies()
   &FN_InitializeAnimTrailBoneIndecies(AnimTrail Selection=1, First Bone or Socket Index=2, Second Bone or Socket Index=3) {enabled=false, module=FN_InitializeAnimTrailBoneIndecies001}
   &FN_InitializeAnimTrail(
    AnimTrailWidth=2
    Lifetime=&FloatFromCurve(
     FloatCurve=DataInterface() {propertyValues {Curve {keys [<at 0 0.825 RCIM_Constant>,<at 800 0.45 RCIM_Constant>]}}}
     CurveIndex=&VectorLength(Vector=$User.OwnerLocalVelocity)
    )
    Position Offset=<0 0 5>
   )
   &AddVelocity(Velocity=&MakeVector(
     X=0
     Y=&MakeCustomFloatFromBool(
      Bool=&SetBoolByFloatComparison(A=&VectorLength(Vector=$User.OwnerLocalVelocity), B=1)
      True Float=0
      False Float=-400
     )
     Z=0
    ))
  ]
  particleUpdateScript [
   &SolveForcesAndVelocity(Clamp Velocity=false, Limit Acceleration=false)
   &FN_UpdateAnimTrail() {module=FN_UpdateAnimTrail001}
   /SetVariables(
    Particles.NotesColor:LinearColor=&Multiply_LinearColorByFloat(
     Linear Color=&ColorFromCurve(ColorCurve=DataInterface() {propertyValues {
        AlphaCurve {keys [<at 0 1 RCIM_Cubic>,<at 1 0 RCIM_Cubic>]}
        BlueCurve {keys [<at 0 1 RCIM_Cubic>,<at 1 1 RCIM_Cubic>]}
        GreenCurve {keys [<at 0 0.4703 RCIM_Cubic>,<at 1 0.2136 RCIM_Cubic>]}
        RedCurve {keys [<at 0 0.1823 RCIM_Cubic>,<at 1 0.7606 RCIM_Cubic>]}
       }})
     Float=&Multiply_Float(
      A=&FloatFromCurve(FloatCurve=DataInterface() {propertyValues {Curve {keys [
           {arriveTangent=10.5633, interpMode=RCIM_Cubic, leaveTangent=10.5633, tangentMode=RCTM_User, time=0, value=0}
           <at 0.25 1 RCIM_Cubic>
           {arriveTangent=-3.9579, interpMode=RCIM_Cubic, leaveTangent=-3.9579, tangentMode=RCTM_User, time=1, value=0}
          ]}}})
      B=&OneMinusFloat(Float=$User.Emote_MerryMaking_Reveal)
     )
    )
    Particles.NotesDP:Vector4f=&MakeVector4(
     X=1
     Y=&Multiply_Float(A=$Emitter.Age, B=-2.5)
     Z=1
     W=&Multiply_Float(A=$Emitter.Age, B=-2.5)
    )
    Particles.NotesDP2:Vector4f=<0 0 0 0>
   ) {isSetParameters=true, module=SetVariables_#3}
  ]
  rendererClasses [&NiagaraRibbonRendererProperties]
  renderers [
   {rendererClass=&NiagaraRibbonRendererProperties}
   {rendererClass=&NiagaraRibbonRendererProperties, rendererIndex=1}
   {rendererClass=&NiagaraRibbonRendererProperties, rendererIndex=2}
  ]
 }
 Trail {
  bEnabled=false
  emitterUpdateScript [
   &EmitterState(Life Cycle Mode=::NewEnumerator1, Loop Behavior=::NewEnumerator1, Loop Duration Mode=::NewEnumerator1, UseLoopDelay=false)
   &SpawnRate(SpawnRate=24, Use Spawn Probability=false)
  ]
  particleSpawnScript [
   &InitializeParticle(
    Lifetime=0.6
    Ribbon Facing Mode=::NewEnumerator1
    Ribbon Width=75
    Ribbon Width Mode=::NewEnumerator1
    UsePositionOffset=false
    Write Material Random=true
   )
   &AddVelocity(Velocity=&MakeVector(
     X=&MakeCustomFloatFromBool(
      Bool=&SetBoolByFloatComparison(A=&VectorLength(Vector=$User.OwnerLocalVelocity), B=1)
      True Float=0
      False Float=400
     )
     Y=0
     Z=0
    ))
  ]
  particleUpdateScript [
   &ParticleState()
   &SolveForcesAndVelocity(Clamp Velocity=false, Limit Acceleration=false)
   &DynamicMaterialParameters(
    Index 0 Param 1=$User.Emote_MerryMaking_Reveal
    Module.Index0Float0Write=true
    Module.Index0Float1Write=true
    Module.Index0Float2Write=true
    Module.Index0Float3Write=true
   )
   &ScaleColor(
    Scale Alpha=&FloatFromCurve(FloatCurve=DataInterface() {propertyValues {Curve {
        defaultValue=1
        keys [
         <at 0 0 RCIM_Cubic>
         {arriveTangentWeight=0.3333, interpMode=RCIM_Cubic, leaveTangentWeight=0.3333, tangentMode=RCTM_User, time=0.125, value=1}
         {arriveTangentWeight=0.3333, interpMode=RCIM_Cubic, leaveTangentWeight=0.3333, tangentMode=RCTM_User, time=0.8, value=0}
        ]
       }}})
    Scale RGB=&Multiply_VectorByFloat(
     Vector=&MakeVectorFromLinearColorRGB(Color=&ColorFromCurve(ColorCurve=DataInterface() {propertyValues {
         AlphaCurve {keys []}
         BlueCurve {keys [<at 0 0.9043 RCIM_Cubic>,<at 1 1 RCIM_Cubic>]}
         GreenCurve {keys [<at 0 1 RCIM_Cubic>,<at 1 0.3742 RCIM_Cubic>]}
         RedCurve {keys [<at 0 0.1441 RCIM_Cubic>,<at 1 0.1656 RCIM_Cubic>]}
        }}))
     Float=15
    )
    ScaleA=true
    ScaleRGB=true
   )
   &FN_Sprite_Attributes(
    Color=&MakeLinearColorFromVectorAndFloat(
     "Vector (RGB)"=&MakeVectorFromLinearColorRGB(Color=&ColorFromCurve(ColorCurve=DataInterface() {propertyValues {
         AlphaCurve {keys []}
         BlueCurve {keys [<at 0 0.9043 RCIM_Cubic>,<at 1 1 RCIM_Cubic>]}
         GreenCurve {keys [<at 0 1 RCIM_Cubic>,<at 1 0.3742 RCIM_Cubic>]}
         RedCurve {keys [<at 0 0.1441 RCIM_Cubic>,<at 1 0.1656 RCIM_Cubic>]}
        }}))
     "Float (Alpha)"=&FloatFromCurve(
      FloatCurve=DataInterface() {propertyValues {Curve {
         defaultValue=1
         keys [
          <at 0 0 RCIM_Cubic>
          {arriveTangentWeight=0.3333, interpMode=RCIM_Cubic, leaveTangentWeight=0.3333, tangentMode=RCTM_User, time=0.125, value=1}
          {arriveTangentWeight=0.3333, interpMode=RCIM_Cubic, leaveTangentWeight=0.3333, tangentMode=RCTM_User, time=0.8, value=0}
         ]
        }}}
      Scale Curve=1.25
     )
    )
    Create Color Attribute=true
    Write to Module Namespace=true
   )
   &FN_Sprite_Attributes(
    Color=&MakeLinearColorFromVectorAndFloat(
     "Vector (RGB)"=&MakeVectorFromLinearColorRGB(Color=&ColorFromCurve(ColorCurve=DataInterface() {propertyValues {
         AlphaCurve {keys []}
         BlueCurve {keys [<at 0 0.9043 RCIM_Cubic>,<at 1 1 RCIM_Cubic>]}
         GreenCurve {keys [<at 0 1 RCIM_Cubic>,<at 1 0.3742 RCIM_Cubic>]}
         RedCurve {keys [<at 0 0.1441 RCIM_Cubic>,<at 1 0.1656 RCIM_Cubic>]}
        }}))
     "Float (Alpha)"=&FloatFromCurve(
      FloatCurve=DataInterface() {propertyValues {Curve {
         defaultValue=1
         keys [
          <at 0 0 RCIM_Cubic>
          {arriveTangentWeight=0.3333, interpMode=RCIM_Cubic, leaveTangentWeight=0.3333, tangentMode=RCTM_User, time=0.125, value=1}
          {arriveTangentWeight=0.3333, interpMode=RCIM_Cubic, leaveTangentWeight=0.3333, tangentMode=RCTM_User, time=1, value=0}
         ]
        }}}
      Scale Curve=0.8
     )
    )
    Create Color Attribute=true
    Write to Module Namespace=true
   ) {module=FN_Sprite_Attributes001}
   /SetVariables(Particles.LineRibbonWidth:NiagaraFloat=&FloatFromCurve(FloatCurve=DataInterface() {propertyValues {Curve {keys [<at 0 1 RCIM_Cubic>,<at 1 0.8 RCIM_Cubic>]}}}, Scale Curve=75)) {isSetParameters=true}
   /SetVariables(Particles.GlowRibbonWidth:NiagaraFloat=&FloatFromCurve(
     FloatCurve=DataInterface() {propertyValues {Curve {
        defaultValue=1
        keys [
         {interpMode=RCIM_Cubic, tangentMode=RCTM_User, time=0, value=1}
         {arriveTangent=-3.452, interpMode=RCIM_Cubic, leaveTangent=-3.452, tangentMode=RCTM_User, time=1, value=0}
        ]
       }}}
     Scale Curve=100
    )) {isSetParameters=true}
  ]
  rendererClasses [&NiagaraRibbonRendererProperties]
  renderers [
   {rendererClass=&NiagaraRibbonRendererProperties}
   {rendererClass=&NiagaraRibbonRendererProperties, rendererIndex=1}
   {rendererClass=&NiagaraRibbonRendererProperties, rendererIndex=2}
  ]
 }
 Wake {
  emitterSpawnScript [/SetVariables(
    Emitter.Lines Color Center:LinearColor=<rgba 0.1425 0.341 1 1>
    Emitter.Lines Color Inner:LinearColor=<rgba 0.6177 0.8635 1 1>
    Emitter.Lines Color Outer:LinearColor=<rgba 0.6843 0.6064 1 1>
   ) {isSetParameters=true}]
  emitterUpdateScript [
   &EmitterState(
    Inactive Response=::NewEnumerator1
    Life Cycle Mode=::NewEnumerator1
    Loop Behavior=::NewEnumerator1
    Loop Duration Mode=::NewEnumerator1
    UseLoopDelay=false
   )
   &SpawnBurst_Instantaneous(Use Spawn Probability=false, UseLoopCountLimit=false)
   &SpawnBurst_Instantaneous(Spawn Group=1, Use Spawn Probability=false, UseLoopCountLimit=false) {module=SpawnBurst_Instantaneous001}
   &SpawnBurst_Instantaneous(Spawn Group=3, Use Spawn Probability=false, UseLoopCountLimit=false) {module=SpawnBurst_Instantaneous002}
   &SpawnBurst_Instantaneous(Spawn Group=2, Use Spawn Probability=false, UseLoopCountLimit=false) {module=SpawnBurst_Instantaneous003}
  ]
  particleSpawnScript [
   &InitializeParticle(
    Color=<rgba 2 2 2 1>
    Lifetime=$Emitter.CurrentLoopDuration
    Mesh Scale=&Multiply_VectorByFloat(Vector=<2.25 1.5 1>, Float=2.25)
    Mesh Scale Mode=::NewEnumerator4
    Position Offset Coordinate Space=::Local
    Sprite Size Mode=::NewEnumerator3
    UsePositionOffset=true
    Write Material Random=true
   )
   &InitialMeshOrientation(
    Mesh Axis To Orient=<0 0 -1>
    Mesh Orientation Mode=::NewEnumerator2
    Rotation=<0 0 -0.25>
    Rotation Coordinate Space=::NewEnumerator2
    Use Rotation Vector=true
   ) {module=InitialMeshOrientation001}
   /SetVariables(Particles.VisibilityTag=&SpawnGroup()) {isSetParameters=true}
  ]
  particleUpdateScript [
   &ParticleState(Kill Particles When Lifetime Has Elapsed=false)
   &SkeletalMeshLocation(Bone Sampling Mode=::NewEnumerator4, MaskSpawnGroup=false, Sampled Position Offset=<40 20 0>, Skeletal Mesh=$System.Skeletal Mesh Reader)
   &SolveForcesAndVelocity(Clamp Velocity=false, Limit Acceleration=false)
   /SetVariables(Particles.Position=&AddVectorToPosition(Position=$Engine.Emitter.SimulationPosition, Vector=<8 -3.5 -1.5>)) {enabled=false, isSetParameters=true}
   &DynamicMaterialParameters(
    Index 0 Param 1=0.8
    Index 0 Param 2=&Multiply_Float(
     A=0.71
     B=&Waveform("Waveform [1]"=::NewEnumerator2, "[1] Frequency"=1, "[1] Compound Frequency"=1, "[1] Amplitude Min/Max"=<-0.5 1.5>)
    )
    Index 0 Param 3=6
    Module.Index0Float0Write=true
    Module.Index0Float1Write=true
    Module.Index0Float2Write=true
    Module.Index0Float3Write=true
   )
   &ScaleColor(
    Scale Alpha=&FloatFromCurve(
     FloatCurve=DataInterface() {propertyValues {Curve {
        defaultValue=1
        keys [
         {arriveTangent=2.8635, interpMode=RCIM_Cubic, leaveTangent=2.8635, tangentMode=RCTM_User, time=0, value=0}
         {arriveTangent=0.0224, interpMode=RCIM_Cubic, leaveTangent=0.0224, tangentMode=RCTM_User, time=1, value=1}
        ]
       }}}
     CurveIndex=$Emitter.LoopedAge
     Scale Curve=0.4
    )
    Scale RGB=&Multiply_Vector(
     A=&MakeVectorFromLinearColorRGB(Color=<rgba 0.3638 1 0.8583 1>)
     B=&MakeVectorFromLinearColorRGB(Color=$System.TODColorMultiplier)
    )
    ScaleA=true
    ScaleRGB=true
   )
   /SetVariables(
    Particles.WakeAuraColor:LinearColor=/Niagara/DynamicInputs/TypeConversions/MakeLinearColorFromFloat(Value:NiagaraFloat=&Multiply_Float(A=1, B=&OneMinusFloat(Float=$User.Emote_MerryMaking_Reveal)))
    Particles.WakeAuraDP:Vector4f=&MakeVector4(X=-0.5, Y=3, Z=0.5, W=2)
   ) {isSetParameters=true}
  ]
  rendererClasses [&NiagaraMeshRendererProperties]
  renderers [
   {rendererClass=&NiagaraMeshRendererProperties, rendererIndex=0}
   {rendererClass=&NiagaraMeshRendererProperties, rendererIndex=1}
   {rendererClass=&NiagaraMeshRendererProperties, rendererIndex=2}
   {rendererClass=&NiagaraMeshRendererProperties, rendererIndex=3}
  ]
 }
 WakeLines {
  emitterSpawnScript [/SetVariables(Emitter.ParticlesPerTrail=30) {isSetParameters=true}]
  emitterUpdateScript [
   &EmitterState(Life Cycle Mode=::NewEnumerator1, Loop Behavior=::NewEnumerator1, Loop Duration Mode=::NewEnumerator1, UseLoopDelay=false)
   &SpawnBurst_Instantaneous(
    Spawn Count=&Multiply_Int(A=$Emitter.ParticlesPerTrail, B=2)
    Use Spawn Probability=false
    UseLoopCountLimit=true
   )
  ]
  particleSpawnScript [
   &PartitionParticles(
    Normalized Index In Partition Includes One=true
    Particles Per Partition:NiagaraInt32=$Emitter.ParticlesPerTrail
    Partition Mode:ENiagara_PartitionMode_v2=::NewEnumerator1
    Write Partition Index to Ribbon Link Order:ENiagara_AttributeSamplingApplyOutput=::NewEnumerator0
    Write Partition to Ribbon ID=::NewEnumerator0
   )
   &InitializeParticle(
    Color=<rgba 0.2503 1 0.8849 0.6>
    Lifetime=5
    Mesh Scale=<0.6 0.4 1>
    Mesh Scale Mode=::NewEnumerator4
    UsePositionOffset=false
    Write Material Random=true
   )
   /SetVariables() {isSetParameters=true}
   /SetVariables(Particles.GizmoScale:Vector3f=<5 5 5>) {isSetParameters=true}
   /SetVariables(Particles.NormalizedParticleIndex:NiagaraFloat=&Power(a=&OneMinusFloat(Float=$Particles.PartitionParticles.Normalized Index In Partition), b=3)) {isSetParameters=true}
  ]
  particleUpdateScript [
   &ParticleState(Kill Particles When Lifetime Has Elapsed=false, Let Infinitely Lived Particles Die When Emitter Deactivates=true)
   /SetVariables(
    Particles.RibbonFacing=&TransformVector(Vector=<0 0 1>, Source Space=::World, Destination Space=::Local)
    Particles.RibbonWidth=&Lerp_Float(
     A=65
     B=8
     Alpha=&FloatFromCurve(
      FloatCurve=DataInterface() {propertyValues {Curve {keys [{arriveTangent=-2.079, interpMode=RCIM_Cubic, leaveTangent=-2.079, tangentMode=RCTM_User, time=0, value=1}, <at 0.7 0 RCIM_Cubic>]}}}
      CurveIndex=$Particles.PartitionParticles.Normalized Index In Partition
     )
    )
   ) {isSetParameters=true}
   /SetVariables(Particles.PreviousPos:NiagaraPosition=$Particles.Position) {isSetParameters=true}
   &SkeletalMeshLocation(
    Bone Sampling Mode=::NewEnumerator4
    MaskSpawnGroup=false
    Sampled Position Offset=&Add_Vector(
     A=&MakeVector(
      X=&Add_Float(A=65, B=&Multiply_Float(A=$Particles.PartitionParticles.Normalized Index In Partition, B=-300))
      Y=&Add_Float(
       A=5
       B=&Multiply_Float(
        A=&Sine(
         Normalized Angle=&Multiply_Float(
          A=&Add_Float(A=$System.Age, B=&Multiply_Float(A=$Particles.PartitionParticles.Normalized Index In Partition, B=-2))
          B=-1
         )
         Period=1
         Scale=1
         Bias=0
        )
        B=5
       )
      )
      Z=&Add_Float(
       A=&ClampFloat(
        Float=&Multiply_Float(
         A=$Particles.NormalizedParticleIndex
         B=&RemapRange(Input Value To Remap=$Particles.Roadbending, Input Min=-250, Input Max=250, Output Min=-100, Output Max=100)
        )
        Min=-100
        Max=100
       )
       B=&Multiply_Float(
        A=&Sine(
         Normalized Angle=&Multiply_Float(
          A=&Add_Float(A=$System.Age, B=&Multiply_Float(A=$Particles.PartitionParticles.Normalized Index In Partition, B=-2))
          B=&MergeHelper_FloatSwitch(0=1, 1=-1, 2=?, 3=?, Index=$Particles.PartitionParticles.Partition, Num=::NewEnumerator0)
         )
         Period=1
         Scale=1
         Bias=0
        )
        B=&FloatFromCurve(
         FloatCurve=DataInterface() {propertyValues {Curve {
            defaultValue=1
            keys [
             {arriveTangent=3.7882, interpMode=RCIM_Cubic, leaveTangent=3.7882, tangentMode=RCTM_User, time=0, value=0}
             <at 0.499 1 RCIM_Cubic>
             {arriveTangent=3.6799, interpMode=RCIM_Cubic, leaveTangent=3.6799, tangentMode=RCTM_User, time=0.5, value=0}
             {interpMode=RCIM_Cubic, tangentMode=RCTM_User, time=1, value=1}
            ]
           }}}
         CurveIndex=&ReturnNormalizedExecIndex()
         Scale Curve=15
        )
       )
      )
     )
     B=&Add_Vector(
      A=&AddVectorToPosition(Position=Linked() {linkedVariable {name=System.MasterRoadOffset, type=Vector3f}}, Vector=<0 0 0>)
      B=&MakeVector(
       X=0
       Y=0
       Z=&Lerp_Float(
        A=&MergeHelper_FloatSwitch(0=-80, 1=80, 2=?, 3=?, Index=$Particles.PartitionParticles.Partition, Num=::NewEnumerator0)
        B=&Lerp_Float(
         A=&MergeHelper_FloatSwitch(0=-60, 1=60, 2=?, 3=?, Index=$Particles.PartitionParticles.Partition, Num=::NewEnumerator0)
         B=&MergeHelper_FloatSwitch(0=-35, 1=35, 2=?, 3=?, Index=$Particles.PartitionParticles.Partition, Num=::NewEnumerator0)
         Alpha=&RemapRange(
          Input Value To Remap=$Particles.Roadbending
          Input Min=0
          Input Max=&MergeHelper_FloatSwitch(0=-100, 1=100, 2=?, 3=?, Index=$Particles.PartitionParticles.Partition, Num=::NewEnumerator0)
          Output Min=1
          Output Max=0
         )
        )
        Alpha=&FloatFromCurve(
         FloatCurve=DataInterface() {propertyValues {Curve {keys [{arriveTangent=-1.5648, time=0, value=1}, {leaveTangent=-0.6189, time=1, value=0}]}}}
         CurveIndex=$Particles.PartitionParticles.Normalized Index In Partition
        )
       )
      )
     )
    )
    Skeletal Mesh=$System.Skeletal Mesh Reader
    Use Sampled Mesh Attribute as Ribbon ID=false
   )
   &self:SetRoadBending(CurrentPos:NiagaraPosition=$Particles.Position, MaxBending=5, RestSpeed=1, VelDivider=4) {module=SetRoadBending}
   &ScaleColor(
    Scale RGB=&VectorFromFloat(Value=&RemapRange(
      Input Value To Remap=&Sine(
       Normalized Angle=&Add_Float(
        A=&Multiply_Float(A=$Particles.PartitionParticles.Normalized Index In Partition, B=3)
        B=&Multiply_Float(
         A=&Add_Float(A=$Engine.System.Age, B=&Multiply_Float(A=$System.FN_Integrator002.Value, B=10))
         B=-1
        )
       )
       Period=1
       Scale=1
       Bias=0
      )
      Input Min=-1
      Input Max=1
      Output Min=0.1
      Output Max=1
     ))
    ScaleA=false
    ScaleRGB=true
   )
   &ScaleColor(
    Scale Alpha=&Multiply_Float(
     A=1
     B=&FloatFromCurve(
      FloatCurve=DataInterface() {propertyValues {Curve {keys [<at 0.3 1 RCIM_Cubic>,<at 1 0>]}}}
      CurveIndex=$Particles.Initial.PartitionParticles.Normalized Index In Partition
     )
    )
    Scale RGB=&VectorFromFloat(Value=&Multiply_Float(
      A=&Lerp_Float(A=1, B=6, Alpha=$Particles.NormalizedParticleIndex)
      B=&FloatFromCurve(
       FloatCurve=DataInterface() {propertyValues {Curve {keys [
           <at 0 1 RCIM_Cubic>
           {arriveTangent=5.8588, interpMode=RCIM_Cubic, leaveTangent=5.8588, time=0.2, value=6}
           {arriveTangent=-12.5901, interpMode=RCIM_Cubic, leaveTangent=-12.5901, tangentMode=RCTM_User, time=0.3453, value=3.0232}
           <at 0.8 1 RCIM_Cubic>
          ]}}}
       CurveIndex=$System.Timeline003.NormalizedTime
      )
     ))
    ScaleA=true
    ScaleRGB=true
   ) {module=ScaleColor001}
   &FN_Sprite_Attributes(
    Color=&MultiplyLinearColor(
     Linear Color=<rgba 0.0039 0.0048 0.087 1>
     Scale Factor=&MakeVector4FromVectorAndFloat(
      "Vector (RGB)"=<1 1 1>
      "Float (Alpha)"=&Multiply_Float(
       A=1
       B=&FloatFromCurve(
        FloatCurve=DataInterface() {propertyValues {Curve {keys [{arriveTangent=-4.0505, interpMode=RCIM_Cubic, leaveTangent=-4.0505, tangentMode=RCTM_User, time=0.3, value=1}, <at 1 0>]}}}
        CurveIndex=$Particles.PartitionParticles.Normalized Index In Partition
        Scale Curve=0.6
       )
      )
     )
    )
    Create Color Attribute=true
    Write to Module Namespace=true
   )
   &FN_Sprite_Attributes(
    Color=&MultiplyLinearColor(
     Linear Color=<rgba 0.8049 0.3014 1 1>
     Scale Factor=&MakeVector4FromVectorAndFloat(
      "Vector (RGB)"=<1 1 1>
      "Float (Alpha)"=&Multiply_Float(
       A=1
       B=&FloatFromCurve(
        FloatCurve=DataInterface() {propertyValues {Curve {keys [{arriveTangent=-4.0505, interpMode=RCIM_Cubic, leaveTangent=-4.0505, tangentMode=RCTM_User, time=0.3, value=1}, <at 1 0>]}}}
        CurveIndex=$Particles.PartitionParticles.Normalized Index In Partition
        Scale Curve=0.4
       )
      )
     )
    )
    Create Color Attribute=true
    Write to Module Namespace=true
   ) {module=FN_Sprite_Attributes001}
   &DynamicMaterialParameters(
    Index 0 Param 1=$User.Emote_MerryMaking_Reveal
    Module.Index0Float0Write=true
    Module.Index0Float1Write=false
    Module.Index0Float2Write=false
    Module.Index0Float3Write=false
   )
   &BlendSpace("Position Blend (World <-> Local)":NiagaraFloat=&FloatFromCurve(
     FloatCurve=DataInterface() {propertyValues {Curve {
        defaultValue=1
        keys [
         {arriveTangent=3.7882, interpMode=RCIM_Cubic, leaveTangent=3.7882, tangentMode=RCTM_User, time=0, value=0}
         <at 0.499 1 RCIM_Cubic>
         {arriveTangent=3.6799, interpMode=RCIM_Cubic, leaveTangent=3.6799, tangentMode=RCTM_User, time=0.5, value=0}
         {interpMode=RCIM_Cubic, tangentMode=RCTM_User, time=1, value=1}
        ]
       }}}
     CurveIndex=&ReturnNormalizedExecIndex()
     Translation Contribution:NiagaraFloat=1
     Rotation Contribution:NiagaraFloat=0
    )) {enabled=false}
  ]
  rendererClasses [&NiagaraRibbonRendererProperties, &NiagaraMeshRendererProperties]
  renderers [
   {rendererClass=&NiagaraRibbonRendererProperties}
   {rendererClass=&NiagaraRibbonRendererProperties, rendererIndex=1}
   {rendererClass=&NiagaraRibbonRendererProperties, rendererIndex=2}
   {rendererClass=&NiagaraMeshRendererProperties, rendererIndex=3}
  ]
 }
}
summary {
 systemName=NS_Emote_MerryMaking_Trail
 emitters [
  {bEnabled=false, emitterName=Trail, rendererClasses [&NiagaraRibbonRendererProperties]}
  {bEnabled=false, emitterName=MusicNotes, rendererClasses [&NiagaraMeshRendererProperties]}
  {emitterName=MusicTrail, rendererClasses [&NiagaraRibbonRendererProperties]}
  {emitterName=Wake, rendererClasses [&NiagaraMeshRendererProperties]}
  {emitterName=WakeLines, rendererClasses [&NiagaraRibbonRendererProperties, &NiagaraMeshRendererProperties]}
  {bEnabled=false, emitterName=FN_Debug, rendererClasses [&NiagaraSpriteRendererProperties]}
  {emitterName=MusicTrail001, rendererClasses [&NiagaraRibbonRendererProperties]}
  {emitterName=Decal, rendererClasses [&NiagaraDecalRendererProperties]}
 ]
 userVariables [
  User.bIsFrontEnd {type=NiagaraBool}
  User.bIsFrontEndPreview {type=NiagaraBool}
  User.Emote_MerryMaking_Light {type=NiagaraFloat, defaultValue=0, description=""}
  User.Emote_MerryMaking_Reveal {type=NiagaraFloat, defaultValue=0, description=""}
  User.Fade {type=NiagaraFloat, defaultValue=0, description=""}
  User.OwnerLocalVelocity {type=Vector3f, description="", defaultValue {struct=Vector3f, value=<0 0 0>}}
  User.ParaGlideLeanAlpha {type=NiagaraFloat, defaultValue=0, description=""}
  User.RibbonIDOffset {type=NiagaraInt32}
  User.Variant {type=NiagaraFloat, defaultValue=0, description=""}
 ]
}
